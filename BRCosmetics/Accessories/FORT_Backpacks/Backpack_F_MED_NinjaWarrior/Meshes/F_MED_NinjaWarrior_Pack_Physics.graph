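# Physics asset for the NinjaWarrior medium female backpack cosmetic, giving its two dangling ropes collision and constrained motion.
# Capsule bodies cover the pack and two rope chains, a five-segment "sh" and six-segment "l", branching from a master rope bone on the pack.
# Upper segments mainly swing on one axis with twist locked, while lower segments gain slight side swing.
# The pack-to-root constraint leaves all angular motion free.
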PhysicsAsset /BRCosmetics/Accessories/FORT_Backpacks/Backpack_F_MED_NinjaWarrior/Meshes/F_MED_NinjaWarrior_Pack_Physics
bodies {
 Base_rope_sh [{shapeType=Capsule, length=0.505, radius=0.505, shapeName=Base_rope_sh_capsule, rotation=<rot 0 -90 0>}]
 Mstr_rope [{shapeType=Capsule, length=0.505, radius=0.505, shapeName=Mstr_rope_capsule, rotation=<rot 0 -90 0>}]
 base_rope_l [{shapeType=Capsule, length=0.505, radius=0.505, shapeName=base_rope_l_capsule, rotation=<rot 0 -90 0>}]
 dyn_rope_l_01 [{
   shapeType=Capsule
   length=2.2482
   radius=1.1207
   shapeName=dyn_rope_l_01_capsule
   center=<2.3419 -0.048 -0.2263>
   rotation=<rot 90 179.9473 179.9467>
  }]
 dyn_rope_l_02 [{
   shapeType=Capsule
   length=2.0898
   radius=1.0372
   shapeName=dyn_rope_l_02_capsule
   center=<2.4898 0.0738 -0.1682>
   rotation=<rot 90 179.9805 179.9799>
  }]
 dyn_rope_l_03 [{
   shapeType=Capsule
   length=2.9215
   radius=1.5729
   shapeName=dyn_rope_l_03_capsule
   center=<2.0728 0.0254 -0.1281>
   rotation=<rot 90 179.9805 179.9799>
  }]
 dyn_rope_l_04 [{shapeType=Capsule, length=2.1237, radius=1.4689, shapeName=dyn_rope_l_04_capsule, rotation=<rot 90 179.9854 179.9847>}]
 dyn_rope_l_05 [{shapeType=Capsule, length=1.8147, radius=1.8199, shapeName=dyn_rope_l_05_capsule, center=<0.3184 0 0>, rotation=<rot -90 179.9902 -179.9892>}]
 dyn_rope_l_06 [{shapeType=Capsule, length=2.968, radius=1.6127, shapeName=dyn_rope_l_06_capsule, center=<2.7296 0 0>, rotation=<rot 90 179.9978 179.9972>}]
 dyn_rope_sh_01 [{
   shapeType=Capsule
   length=2.2953
   radius=1.0156
   shapeName=dyn_rope_sh_01_capsule
   center=<2.7899 -0.0048 -0.1254>
   rotation=<rot -90 -90 90.0002>
  }]
 dyn_rope_sh_02 [{
   shapeType=Capsule
   length=1.9425
   radius=0.9392
   shapeName=dyn_rope_sh_02_capsule
   center=<1.726 -0.0068 0.1326>
   rotation=<rot -90 -91.1471 91.1473>
  }]
 dyn_rope_sh_03 [{
   shapeType=Capsule
   length=2.1087
   radius=0.9424
   shapeName=dyn_rope_sh_03_capsule
   center=<1.0313 0.0021 0.0712>
   rotation=<rot -90 -91.1471 91.1473>
  }]
 dyn_rope_sh_04 [{
   shapeType=Capsule
   length=2.1165
   radius=0.9607
   shapeName=dyn_rope_sh_04_capsule
   center=<1.1017 0.1005 -0.0123>
   rotation=<rot -90 -91.1471 91.1473>
  }]
 dyn_rope_sh_05 [{
   shapeType=Capsule
   length=2.7271
   radius=1.3672
   shapeName=dyn_rope_sh_05_capsule
   center=<2.1687 0 -0.1897>
   rotation=<rot -90 -91.1471 91.1473>
  }]
 pack [
  {
   shapeType=Capsule
   length=20.022
   radius=7.953
   shapeName=pack_capsule
   center=<-0.7517 -9.6636 0.465>
   rotation=<rot 11.958 -79.4063 -62.9868>
  }
  {shapeType=Capsule, length=4, radius=8, shapeName=pack_capsule, center=<-7.0148 2.08 -10.2483>, rotation=<rot 0.0002 85 85>}
 ]
 root [{shapeType=Capsule, length=0.505, radius=0.505, shapeName=root_capsule, rotation=<rot 0 -90 0>}]
}
constraints [
 {
  bone1Name=pack
  bone2Name=root
  swing1LimitDegrees=45
  swing1Motion=Free
  swing2LimitDegrees=45
  swing2Motion=Free
  twistLimitDegrees=45
  twistMotion=Free
 }
 {bone1Name=Mstr_rope, bone2Name=pack, swing1LimitDegrees=45, swing2LimitDegrees=0, twistLimitDegrees=45, twistMotion=Locked}
 {
  bone1Name=Base_rope_sh
  bone2Name=Mstr_rope
  swing1LimitDegrees=45
  swing2LimitDegrees=45
  swing2Motion=Locked
  twistLimitDegrees=45
  twistMotion=Locked
 }
 {bone1Name=dyn_rope_sh_01, bone2Name=Base_rope_sh, swing1LimitDegrees=33, swing2LimitDegrees=0, twistLimitDegrees=5, twistMotion=Locked}
 {bone1Name=dyn_rope_sh_02, bone2Name=dyn_rope_sh_01, swing1LimitDegrees=33, swing2LimitDegrees=0, twistLimitDegrees=5, twistMotion=Locked}
 {bone1Name=dyn_rope_sh_03, bone2Name=dyn_rope_sh_02, swing1LimitDegrees=33, swing2LimitDegrees=10, twistLimitDegrees=5}
 {bone1Name=dyn_rope_sh_04, bone2Name=dyn_rope_sh_03, swing1LimitDegrees=33, swing2LimitDegrees=10, twistLimitDegrees=5}
 {bone1Name=dyn_rope_sh_05, bone2Name=dyn_rope_sh_04, swing1LimitDegrees=33, swing2LimitDegrees=10, twistLimitDegrees=5}
 {
  bone1Name=base_rope_l
  bone2Name=Mstr_rope
  swing1LimitDegrees=45
  swing2LimitDegrees=45
  swing2Motion=Locked
  twistLimitDegrees=45
  twistMotion=Locked
 }
 {bone1Name=dyn_rope_l_01, bone2Name=base_rope_l, swing1LimitDegrees=40, swing2LimitDegrees=0, twistLimitDegrees=0, twistMotion=Locked}
 {bone1Name=dyn_rope_l_02, bone2Name=dyn_rope_l_01, swing1LimitDegrees=40, swing2LimitDegrees=0, twistLimitDegrees=0, twistMotion=Locked}
 {bone1Name=dyn_rope_l_03, bone2Name=dyn_rope_l_02, swing1LimitDegrees=40, swing2LimitDegrees=10, twistLimitDegrees=0}
 {bone1Name=dyn_rope_l_04, bone2Name=dyn_rope_l_03, swing1LimitDegrees=40, swing2LimitDegrees=10, twistLimitDegrees=0}
 {bone1Name=dyn_rope_l_05, bone2Name=dyn_rope_l_04, swing1LimitDegrees=40, swing2LimitDegrees=10, twistLimitDegrees=0}
 {bone1Name=dyn_rope_l_06, bone2Name=dyn_rope_l_05, swing1LimitDegrees=40, swing2LimitDegrees=10, twistLimitDegrees=0}
]
